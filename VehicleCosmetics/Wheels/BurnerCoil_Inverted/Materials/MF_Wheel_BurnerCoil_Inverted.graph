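# Material function for the inverted BurnerCoil cosmetic wheel that sets its animated, color-tinted emissive glow through material attributes.
# Rim and tire mask channels each drive a glow modulated by a panning shared distortion texture, plus an extra green-channel glow sharing one multiplier.
# One color parameter tints both parts, and red vertex color blends from rim to tire emissive, so one function covers the whole wheel.
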MaterialFunction /VehicleCosmetics/Wheels/BurnerCoil_Inverted/Materials/MF_Wheel_BurnerCoil_Inverted
graph {nodes [
  n1 = VectorParameter() {group=00_Base, parameterName=Color, sortPriority=0, defaultValue=<rgba 0.0482 0.1779 1 1>}
  n2 = AppendVector(A=@n1.RGB, B=@n1.RGB)
  n3 = TextureSampleParameter2D() {
   group=00_Base
   parameterName=Rim Mask
   sortPriority=1
   texture=/VehicleCosmetics/Wheels/BurnerCoil_Inverted/Textures/T_BurnercoilInverted_Rim_M
  }
  n4 = Panner(Coordinate=TextureCoordinate() {vTiling=0.1}) {speedX=0.1}
  n5 = Multiply(
   A=ScalarParameter() {defaultValue=4, group=00_Base, parameterName=Rim Emissive Line Glow, sortPriority=4}
   B=Multiply(A=TextureSample(UVs=@n4) {texture=/VehicleCosmetics/Wheels/_SharedTextures/FX/Distort_RGB_T}.B) {constB=2}
  ) {constB=2}
  n6 = Add(A=@n3.B, B=Multiply(A=Multiply(A=@n3.B, B=@n5)) {constB=2})
  n7 = ScalarParameter() {defaultValue=2, group=00_Base, parameterName=ExtraRimEmissiveMultiplier, sortPriority=6}
  n8 = Add(
   A=Multiply(A=Multiply(A=@n2, B=@n6), B=ScalarParameter() {defaultValue=3, group=00_Base, parameterName=RimEmissiveBrightness, sortPriority=5})
   B=Multiply(A=@n2, B=Multiply(A=@n3.G, B=@n7))
  )
  n9 = TextureSampleParameter2D() {
   group=00_Base
   parameterName=Tire Mask
   sortPriority=2
   texture=/VehicleCosmetics/Wheels/BurnerCoil_Inverted/Textures/T_BurnercoilInverted_Tire_M
  }
  n10 = TextureSample(UVs=Panner(Coordinate=TextureCoordinate() {uTiling=2, vTiling=0.1}) {speedX=-0.3, speedY=-0.1}) {texture=&Distort_RGB_T}
  n11 = Add(
   A=Multiply(A=@n10.G, B=ScalarParameter() {defaultValue=4, group=00_Base, parameterName=Tread Emissive 1, sortPriority=7})
   B=Multiply(A=@n10.G) {constB=2}
  )
  n12 = Add(B=Multiply(A=@n9.R, B=Add(A=@n9.R, B=@n11)))
  n13 = Add(
   A=Multiply(A=Multiply(A=@n2, B=@n12), B=ScalarParameter() {defaultValue=3, group=00_Base, parameterName=Tread Emissive Multiplier, sortPriority=8})
   B=Multiply(A=@n2, B=Multiply(A=@n7, B=@n9.G))
  )
  n14 = VertexColor()
  FunctionOutput(None=SetMaterialAttributes!(MaterialAttributes, EmissiveColor=LinearInterpolate(A=@n8, B=@n13, Alpha=@n14.R)) {attributeSetTypes [#1]})
  Multiply(A=ScalarParameter(), B=@n9.R)
 ]}
