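# Translucent unlit particle material for the inverted BurnerCoil wheel cosmetic, producing a distorted, smoke-like effect on sprites, ribbons and mesh particles.
# A panning flow normal map offsets the UVs of cloud and smoke-mask textures, whose product is sharpened by a power.
# Dynamic parameters drive flow speed, distortion and sharpness.
# Particle alpha scales the result, which tints with particle color and feeds a depth fade for soft intersections.
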Material /VehicleCosmetics/Wheels/BurnerCoil_Inverted/Materials/M_Wheel_BurnerCoil_Inverted_PS
graph {
 properties {
  bAutomaticallySetUsageInEditor=false
  bUsedWithMeshParticles=true
  bUsedWithNiagaraMeshParticles=true
  bUsedWithNiagaraRibbons=true
  bUsedWithNiagaraSprites=true
  bUsedWithParticleSprites=true
  blendMode=BLEND_Translucent
  shadingModel=MSM_Unlit
  usedShadingModels=Unlit
 }
 nodes [
  n1 = Time()
  n2 = DynamicParameter() {paramNames [Speed,Distort,Sharpness,Param4]}
  n3 = TextureSample(UVs=Panner(Coordinate=TextureCoordinate() {uTiling=0.7, vTiling=0.7}, Time=Multiply(A=@n1, B=@n2.Speed)) {speedX=-0.1, speedY=0.1}) {samplerType=SAMPLERTYPE_Normal, texture=/VehicleCosmetics/Wheels/_SharedTextures/FX/quasarflow}
  n4 = Add(A=Multiply(A=ComponentMask(None=@n3.RGB) {g=true, r=true}, B=@n2.Distort), B=TextureCoordinate())
  n5 = TextureSample(UVs=@n4) {samplerSource=SSM_Wrap_WorldGroupSettings, texture=/VehicleCosmetics/Wheels/_SharedTextures/FX/Cloud_T}
  n6 = TextureSample(UVs=@n4) {samplerSource=SSM_Wrap_WorldGroupSettings, texture=/VehicleCosmetics/Wheels/_SharedTextures/FX/SmokeMask_T}
  n7 = ParticleColor()
  n8 = Multiply(A=Power(Base=Multiply(A=@n5.RGB, B=@n6.RGB), Exp=@n2.Sharpness), B=@n7.A)
  DepthFade(Opacity=@n8) {fadeDistanceDefault=8}
  Multiply(A=@n7.RGB, B=@n8)
 ]
}
